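# Anim Blueprint for the masked SummitReedGrit male head on the Fortnite 3L head skeleton, choosing whether mouth expressions are muted.
# When character customization loads, it reads the "Hide Element 02" scalar from a dynamic material instance on the Face part.
# A value above zero clears the facial curve mute expression; otherwise it assigns the shared mouth-mute curve asset, tying facial animation to material state.
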AnimBlueprint /BRCosmetics/Characters/Player/Male/Medium/Heads/M_MED_SummitReedGrit_Head/Meshes/M_MED_SummitReedGrit_Head_Masked_AnimBP
target_skeleton=/Game/Characters/Player/Common/Fortnite_Base_Head/Fortnite_Base_3L_Head_Skeleton
graphs [EventGraph {nodes [
   n1 = Event OnCharacterCustomizationLoaded()
   > n2 = Get Dynamic Material Instance(SkeletalMeshComponent=Get Array Item with EFort Custom Part Type(Array=@n1.CustomMeshParts, PartType=Face), ElementIndex=1)
   > n3 = Get Scalar Parameter Value!(execute, self:Material Instance Dynamic Object Reference=@n2, ParameterName:Name=Hide Element 02) -> (then, ReturnValue:float) {is=title K2Node_CallFunction}
   > Set facial_curve_mute_expression(facial_curve_mute_expression:Curve Expressions Data Asset Object Reference=Select(
     Option 0:Curve Expressions Data Asset Object Reference=/BRCosmetics/Characters/Player/Common/Fortnite_Base_Head/Facials/CurveMappings/FN_3L_Mouth_Curve_Mute_Expr
     Option 1:Curve Expressions Data Asset Object Reference
     Index="float > float"(A=@n3)
    ) -> (ReturnValue:Curve Expressions Data Asset Object Reference)) -> (Output_Get:Curve Expressions Data Asset Object Reference)
  ]}]
variables []
